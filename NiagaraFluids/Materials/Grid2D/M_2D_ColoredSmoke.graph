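# Translucent, two-sided unlit material that displays Niagara Fluids 2D colored smoke on sprites or ribbons.
# The SimGrid texture parameter supplies RGBA, which a MaterialX Premult function premultiplies by alpha to form emissive color.
# A Substrate legacy conversion node carries the unlit shading, with opacity left unconnected.
# Depth testing is disabled and responsive AA is enabled.
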Material /NiagaraFluids/Materials/Grid2D/M_2D_ColoredSmoke
graph {
 properties {
  bDisableDepthTest=true
  bEnableResponsiveAA=true
  bUsedWithNiagaraRibbons=true
  bUsedWithNiagaraSprites=true
  bUsedWithParticleSprites=true
  blendMode=BLEND_Translucent
  pixelDepthOffsetMode=PDOM_Legacy
  shadingModel=MSM_FromMaterialExpression
  twoSided=true
  usedShadingModels=Unlit
 }
 nodes [
  MaterialExpressionPremult_0 = MaterialXPremult!(None=TextureSampleParameter2D() {parameterName=SimGrid}.RGBA)
  MaterialExpressionStrataLegacyConversion_0 = SubstrateShadingModels!(
   BaseColor
   Metallic
   Specular
   Roughness
   Anisotropy
   Emissive Color=@MaterialExpressionPremult_0
   Normal
   Tangent
   Subsurface Color
   Custom0
   Custom1
   Opacity
   Thin Translucent Transmittance Color
   Water Scattering Coefficients
   Water Absorption Coefficients
   Water Phase G
   Color Scale BehindWater
   Unused
   Custom Tangent
   Thin Translucent Surface Coverage
   Single Shading Model=ShadingModel() {shadingModel=MSM_Unlit}
  )
 ]
}
