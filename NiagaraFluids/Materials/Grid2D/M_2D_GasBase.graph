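# Base unlit translucent material that renders Niagara Fluids 2D gas simulations as smoke and fire, with depth-faded opacity.
# SimGrid density and temperature are remapped through optional ramps to drive opacity.
# Emissive adds smoke albedo lit by LightingGrid sun and environment channels, blackbody or ramp fire color, and multiple scattering from LightingGrid blue, masked out where temperature is high.
# A static switch optionally applies inverse eye adaptation.
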Material /NiagaraFluids/Materials/Grid2D/M_2D_GasBase
graph {
 properties {
  bDisableDepthTest=true
  bEnableResponsiveAA=true
  bUsedWithNiagaraRibbons=true
  bUsedWithNiagaraSprites=true
  bUsedWithParticleSprites=true
  blendMode=BLEND_Translucent
  pixelDepthOffsetMode=PDOM_Legacy
  shadingModel=MSM_Unlit
  twoSided=true
  usedShadingModels=Unlit
 }
 nodes [
  n1 = TextureSampleParameter2D() {parameterName=SimGrid}
  n2 = ScalarParameter() {parameterName=TemperatureOffset}
  n3 = ScalarParameter() {defaultValue=1, parameterName=TemperatureRange}
  n4 = Constant() {r=1}
  n5 = Saturate(None=Divide(A=Subtract(A=@n1.G, B=@n2), B=If(A=@n3, "A > B"=@n3, "A == B"=@n4, "A < B"=@n4)))
  n6 = TextureSampleParameter2D(UVs=@n5) {parameterName=TemperatureOpacityRamp, samplerType=SAMPLERTYPE_LinearColor, texture=/Engine/ArtTools/RenderToTexture/Textures/127grey}
  n7 = StaticSwitchParameter(True=@n6.R, False=@n5) {parameterName=Use Temperature Opacity Curve}
  n8 = ScalarParameter() {defaultValue=1, parameterName=FireOpacityGain}
  n9 = ScalarParameter() {parameterName=DensityOffset}
  n10 = ScalarParameter() {defaultValue=1, parameterName=DensityRange}
  n11 = Constant() {r=1}
  n12 = Saturate(None=Divide(A=Subtract(A=@n1.R, B=@n9), B=If(A=@n10, "A > B"=@n10, "A == B"=@n11, "A < B"=@n11)))
  n13 = Multiply(
   A=StaticSwitchParameter(
    True=TextureSampleParameter2D(UVs=@n12) {parameterName=DensityRamp, samplerType=SAMPLERTYPE_LinearColor, texture=&127grey}.R
    False=@n12
   ) {parameterName=Use Density Curve}
   B=ScalarParameter() {defaultValue=1, parameterName=DensityGain}
  )
  n14 = If(A=@n13, "A > B"=Add(A=@n13, B=ScalarParameter() {parameterName=ConstantExtinction}), "A == B"=@n13, "A < B"=@n13) {constB=0}
  DepthFade(Opacity=Add(A=Multiply(A=@n7, B=@n8), B=@n14), FadeDistance=ScalarParameter() {defaultValue=100, parameterName=Depth Fade})
  n15 = Multiply(
   A=StaticSwitchParameter(
    True=TextureSampleParameterVolume!(UVs, Apply View MipBias) {parameterName=ColorTexture}.RGB
    False=VectorParameter() {parameterName=Smoke Color, defaultValue=<rgba 1 1 1 0>}.RGB
   ) {parameterName=Use Colored Smoke}
   B=ScalarParameter() {defaultValue=0.25, parameterName=Albedo}
  )
  n16 = VectorParameter() {parameterName=SunColor, defaultValue=<rgba 1 0.9218 0.8437 1>}
  n17 = TextureSampleParameter2D() {parameterName=LightingGrid}
  n18 = VectorParameter() {parameterName=EnvironmentColor, defaultValue=<rgba 0.2678 0.4044 0.5599 1>}
  n19 = Multiply(A=Add(A=Multiply(A=@n16.RGB, B=@n17.R), B=Multiply(A=@n18.RGB, B=@n17.G)), B=@n14)
  n20 = Constant() {r=4500}
  n21 = StaticSwitchParameter(
   True=BlackBody!(Temp=Multiply(A=@n20, B=@n5))
   False=TextureSampleParameter2D(UVs=@n5) {parameterName=TemperatureRamp, samplerType=SAMPLERTYPE_LinearColor, texture=&127grey}.RGB
  ) {defaultValue=true, parameterName=Use Blackbody Curve}
  n22 = ScalarParameter() {defaultValue=1, parameterName=FireGain}
  n23 = Constant() {r=4500}
  n24 = Power(Base=@n17.B, Exp=ScalarParameter() {defaultValue=1, parameterName=MultipleScatteringExponent}) {constExponent=1}
  n25 = ScalarParameter() {parameterName=TemperatureOffset}
  n26 = ScalarParameter() {defaultValue=1, parameterName=TemperatureRange}
  n27 = Constant() {r=1}
  n28 = Saturate(None=Divide(A=Subtract(A=@n24, B=@n25), B=If(A=@n26, "A > B"=@n26, "A == B"=@n27, "A < B"=@n27)))
  n29 = StaticSwitchParameter(
   True=BlackBody(Temp=Multiply(A=@n23, B=@n28))
   False=TextureSampleParameter2D(UVs=@n28) {parameterName=TemperatureRamp, samplerType=SAMPLERTYPE_LinearColor, texture=&127grey}.RGB
  ) {defaultValue=true, parameterName=Use Blackbody Curve}
  n30 = ScalarParameter() {defaultValue=0.5, parameterName=MultipleScatteringMask}
  n31 = OneMinus(None=SmoothStep(Max=OneMinus(None=@n30), Value=@n5))
  n32 = Add(
   A=Add(A=Multiply(A=@n15, B=@n19), B=Multiply(A=@n21, B=@n22)) {constB=0}
   B=Multiply(A=Multiply(A=@n29, B=@n31), B=ScalarParameter() {defaultValue=1, parameterName=MultipleScatteringGain}) {constA=123, constB=0.1}
  )
  StaticSwitchParameter(True=EyeAdaptationInverse(LightValueInput=@n32, AlphaInput=Constant() {r=1}).EyeAdaptationInverse, False=@n32) {parameterName=InvertEyeAdaptation}
 ]
}
