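# Translucent unlit two-sided material that renders smoke and fire from Niagara Fluids 2D grid textures.
# SimGrid density and temperature drive curve lookups for opacity and color, with optional blackbody emission.
# LightingGrid adds sun and environment lighting plus a multiple-scattering fire glow, with switchable inverse eye adaptation.
# Depth testing is disabled, so opacity blends a manual scene-depth comparison with a soft fade broken up by animated world-space noise.
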Material /NiagaraFluids/Materials/Grid2D/M_2D_SmokeFire
graph {
 properties {
  bDisableDepthTest=true
  bEnableResponsiveAA=true
  bUsedWithNiagaraRibbons=true
  bUsedWithNiagaraSprites=true
  bUsedWithParticleSprites=true
  blendMode=BLEND_Translucent
  pixelDepthOffsetMode=PDOM_Legacy
  shadingModel=MSM_Unlit
  twoSided=true
  usedShadingModels=Unlit
 }
 nodes [
  n1 = TextureSampleParameter2D() {parameterName=SimGrid}
  n2 = ScalarParameter() {defaultValue=1, group=Temperature, parameterName=TemperatureCurve}
  n3 = Multiply(A=@n1.G, B=@n2)
  n4 = ScalarParameter() {defaultValue=6000, parameterName=TemperatureMaximum}
  n5 = TextureObjectParameter() {parameterName=Curve_Temperature}
  n6 = TextureSample(UVs=Divide(A=@n3, B=@n4) {constB=3000}, Tex=@n5)
  n7 = TextureSample(
   UVs=Multiply(A=@n1.R, B=ScalarParameter() {defaultValue=1, group=Density, parameterName=DensityScale})
   Tex=TextureObjectParameter() {parameterName=Curve_Density}
  )
  n8 = PixelDepth()
  n9 = SceneDepth()
  n10 = Constant()
  n11 = Constant() {r=1}
  n12 = WorldPosition()
  n13 = ScalarParameter() {defaultValue=100, parameterName=DepthFade}
  n14 = Subtract(
   A=MaterialFunctionCall!(World Position=@n12.XYZ) {materialFunction=/Engine/Functions/Engine_MaterialFunctions02/Utility/DepthFromWorldPosition}.Result
   B=Divide(A=@n13, B=ScalarParameter() {defaultValue=10, parameterName=DepthBiasValue}) {constB=10}
  )
  n15 = SceneDepth()
  n16 = Time()
  n17 = Add(
   A=Multiply(A=@n12.XYZ, B=ScalarParameter() {defaultValue=0.05, parameterName=NoiseScale})
   B=MaterialFunctionCall(X=Multiply(A=@n16), Y=Multiply(A=@n16), Z=Multiply(A=@n16) {constB=-1}) {materialFunction=&MakeFloat3}.Result
  )
  n18 = Multiply(A=@n13, B=Multiply(A=Noise!(World Position=@n17, FilterWidth) {levels=3, scale=0.5}) {constB=0.25})
  n19 = Saturate(None=Divide(A=Subtract(A=@n14, B=@n15.R), B=Add(A=@n13, B=@n18)))
  Multiply(
   A=Saturate(None=Add(A=@n6.A, B=@n7.A))
   B=LinearInterpolate(
    A=If(A=@n8.R, B=@n9.R, "A > B"=@n10, "A == B"=@n10, "A < B"=@n11)
    B=OneMinus(None=@n19)
    Alpha=ScalarParameter() {defaultValue=1, parameterName=DepthTestLerp}
   )
  )
  n20 = TextureSampleParameter2D() {parameterName=LightingGrid}
  n21 = Multiply(A=@n20.B, B=@n2) {constA=1}
  n22 = Multiply(A=ScalarParameter() {defaultValue=0.65, parameterName=MultipleScatteringFactor}) {constB=0.65}
  n23 = TextureSample(UVs=Power(Base=Divide(A=@n21, B=@n4) {constB=3000}, Exp=@n22) {constExponent=1}, Tex=@n5)
  n24 = Multiply(
   A=TextureSample(
    UVs=MaterialFunctionCall(X=@n23.A, Y=@n23.A) {materialFunction=&MakeFloat2}.Result
    Tex=@n5
   ).RGB
   B=@n22
  )
  n25 = MaterialFunctionCall(X=@n6.A, Y=@n6.A) {materialFunction=&MakeFloat2}
  n26 = Multiply(
   A=LinearInterpolate(A=TextureSample(UVs=@n25.Result, Tex=@n5).RGB, B=BlackBody!(Temp=@n3), Alpha=ScalarParameter() {defaultValue=1, parameterName=Blackbody})
   B=VectorParameter() {group=Temperature, parameterName=Temperature Tint, defaultValue=<rgba 1 1 1 1>}.RGB
  )
  n27 = ScalarParameter() {defaultValue=1, group=Temperature, parameterName=TemperatureGain}
  n28 = VectorParameter() {parameterName=SunColor, defaultValue=<rgba 1 0.9218 0.8437 1>}
  n29 = VectorParameter() {parameterName=EnvironmentColor, defaultValue=<rgba 0.2678 0.4044 0.5599 1>}
  n30 = MaterialFunctionCall(Float3=Add(A=Multiply(A=@n28.RGB, B=@n20.R), B=Multiply(A=@n29.RGB, B=@n20.G))) {materialFunction=&BreakOutFloat3Components}
  n31 = Divide(A=@n30.R, B=@n7.A)
  n32 = Divide(A=@n30.G, B=@n7.A)
  n33 = Divide(A=@n30.B, B=@n7.A)
  n34 = Multiply(
   A=Multiply(
    A=@n7.A
    B=MaterialFunctionCall(X=@n31, Y=@n32, Z=@n33) {materialFunction=&MakeFloat3}.Result
   )
   B=VectorParameter() {group=Density, parameterName=Density Tint, defaultValue=<rgba 1 1 1 0>}.RGB
  )
  n35 = Add(
   A=Multiply(A=Multiply(A=@n24) {constB=0.1}, B=ScalarParameter() {defaultValue=1, parameterName=MultipleScatteringStrength})
   B=Add(A=Multiply(A=@n26, B=@n27), B=Multiply(A=@n7.RGB, B=@n34))
  )
  StaticSwitchParameter(True=EyeAdaptationInverse(LightValueInput=@n35, AlphaInput=Constant() {r=1}).EyeAdaptationInverse, False=@n35) {parameterName=InvertEyeAdaptation}
 ]
}
