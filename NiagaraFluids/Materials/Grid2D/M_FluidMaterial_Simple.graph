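# Translucent Niagara Fluids material that shades a liquid surface from Niagara-supplied normal-depth and thickness textures.
# A custom jittered radial blur smooths those inputs, with thickness driving color, Beer's-law opacity and refraction strength.
# It layers hue-shifted lighting, refracted scene color, HDRI and skylight Fresnel reflections, and ray-marched shadows (disabled by default), and clips empty pixels by depth.
# Forward and Mobile paths output constant zero color, emissive and opacity.
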Material /NiagaraFluids/Materials/Grid2D/M_FluidMaterial_Simple
graph {
 properties {
  bDisableDepthTest=true
  bEnableResponsiveAA=true
  bTangentSpaceNormal=false
  bUseMaterialAttributes=true
  bUsedWithInstancedStaticMeshes=true
  bUsedWithNanite=true
  bUsedWithNiagaraMeshParticles=true
  bUsedWithNiagaraSprites=true
  blendMode=BLEND_Translucent
  pixelDepthOffsetMode=PDOM_Legacy
  translucencyPass=MTP_BeforeDOF
  twoSided=true
  usedShadingModels=Default Lit
 }
 nodes [
  n1 = Constant()
  n2 = VectorParameter() {group=Shadow, parameterName=ShadowColor, defaultValue=<rgba 0 0 0 1>}
  n3 = MaterialFunctionCall!(Thickness=NamedRerouteUsage(), Depth Scale=ScalarParameter() {defaultValue=5, group=Opacity, parameterName=ThicknessToOpacity_Conc}) {materialFunction=/Engine/Functions/Engine_MaterialFunctions01/Volumetrics/BeersLaw}
  n4 = Multiply(
   A=ScalarParameter() {defaultValue=1, group=Opacity, parameterName=OpacityScale}
   B=StaticSwitchParameter(True=OneMinus(None=@n3.Result), False=Constant() {r=1}) {group=Opacity, parameterName=EnableThicknessToOpacity}
  )
  n5 = LinearInterpolate(B=NamedRerouteUsage(), Alpha=@n4) {constAlpha=0}
  n6 = Ceil(None=NamedRerouteUsage())
  n7 = NamedRerouteUsage()
  n8 = Add(A=NamedRerouteUsage(), B=NamedRerouteUsage())
  n9 = NamedRerouteUsage()
  n10 = LinearInterpolate(A=@n2.RGB, B=LinearInterpolate(A=NamedRerouteUsage(), B=Add(A=@n8, B=@n9), Alpha=@n4) {constAlpha=0}, Alpha=@n6)
  n11 = Add(A=NamedRerouteUsage(), B=NamedRerouteUsage())
  MakeMaterialAttributes!(
   BaseColor=ShadingPathSwitch(Default=LinearInterpolate(A=@n2.RGB, B=@n5, Alpha=@n6), Forward=@n1, Mobile=@n1)
   Metallic
   Specular
   Roughness
   Anisotropy
   EmissiveColor=ShadingPathSwitch(Default=Add(A=@n7, B=@n10), Forward=@n1, Mobile=@n1)
   Opacity=ShadingPathSwitch(Default=Multiply(A=@n11), Forward=@n1, Mobile=@n1)
   OpacityMask
   Normal
   Tangent
   WorldPositionOffset
   SubsurfaceColor
   ClearCoat
   ClearCoatRoughness
   AmbientOcclusion
   Refraction
   CustomizedUVs_0
   CustomizedUVs_1
   CustomizedUVs_2
   CustomizedUVs_3
   CustomizedUVs_4
   CustomizedUVs_5
   CustomizedUVs_6
   CustomizedUVs_7
   PixelDepthOffset
   ShadingModel
   Displacement
  )
  n12 = TextureCoordinate()
  n13 = MaterialFunctionCall(Float3=Constant3Vector() {constant=<rgba 0 5 2 1>}) {materialFunction=&BreakOutFloat3Components}
  n14 = Multiply(A=Min(A=@n12, B=OneMinus(None=@n12)), B=@n13.G)
  n15 = MaterialFunctionCall(Float2=Power(Base=Saturate(None=@n14), Exp=@n13.B)) {materialFunction=&BreakOutFloat2Components}
  Multiply(A=@n15.R, B=@n15.R)
  NamedRerouteDeclaration(None=TextureObject() {samplerType=SAMPLERTYPE_LinearGrayscale, texture=/Engine/EngineMaterials/Good64x64TilingNoiseHighFreq}) {name=RandomTex, nodeColor=<rgba 0 0.5078 1 1>}
  MaterialExpressionTextureObjectParameter_0 = TextureObjectParameter() {group=FromNiagara, parameterName=NormalDepthTex}
  MaterialExpressionTextureObjectParameter_1 = TextureObjectParameter() {group=FromNiagara, parameterName=ThicknessTex}
  MaterialExpressionTextureCoordinate_0 = TextureCoordinate()
  n16 = MaterialFunctionCall(Float2=Divide(B=ParticleSize()) {constA=300, constB=300}) {materialFunction=&BreakOutFloat2Components}
  MaterialExpressionMultiply_3 = Multiply(A=ScalarParameter() {defaultValue=0.012, group=Filter, parameterName=Filter_BlurDistance}, B=Max(A=@n16.R, B=@n16.R))
  MaterialExpressionConstant_34 = Constant() {r=8}
  MaterialExpressionConstant_35 = Constant() {r=6}
  MaterialExpressionConstant_5 = Constant() {r=0.618}
  MaterialExpressionNamedRerouteUsage_26 = NamedRerouteUsage()
  n17 = Custom!(
   NormalDepthTex=@MaterialExpressionTextureObjectParameter_0
   ThicknessTex=@MaterialExpressionTextureObjectParameter_1
   UV=@MaterialExpressionTextureCoordinate_0
   Distance=@MaterialExpressionMultiply_3
   DistanceSteps=@MaterialExpressionConstant_34
   RadialSteps=@MaterialExpressionConstant_35
   RadialOffset=@MaterialExpressionConstant_5
   RandomTex=@MaterialExpressionNamedRerouteUsage_26
  ) {
   code="\r\n##define TWOPI 6.283185\r\n\r\nfloat2 NewUV = UV;\r\n\r\nfloat StepSize = Distance / (int) DistanceSteps;\r\nfloat CurDistance=0;\r\nfloat2 CurOffset=0;\r\nfloat SubOffset = 0;\r\n\r\nfloat4 AccumNormalDepth=0;\r\nfloat AccumThickness=0;\r\nfloat Accumdist=0;\r\n\r\nfloat2 ScreenUV = MaterialFloat2(ScreenAlignedPosition(Parameters.ScreenPosition).xy);\r\nfloat2 ScenePixels=View.BufferSizeAndInvSize.xy*ScreenUV;\r\nScenePixels+=View.TemporalAAParams.r;\r\nfloat2 RandomSamp = ((uint)(ScenePixels.x) + 2 * (uint)(ScenePixels.y)) % 5;\r\nRandomSamp/=5;RandomSamp-=0.5;RandomSamp*=0.5;\r\nfloat TempAARotation = 1;\r\nfloat TempAADistance = 1;\r\nTempAARotation*=RandomSamp.x;\r\nTempAADistance*=StepSize*RandomSamp.x;\r\n\r\nint i=0;\r\nif (DistanceSteps < 1)\r\n{\r\nThicknessOutput =  Texture2DSample(ThicknessTex,ThicknessTexSampler,UV);\r\nreturn Texture2DSample(NormalDepthTex,NormalDepthTexSampler,UV);\t\t\r\n}\r\nelse\r\n{\r\nwhile ( i < (int) DistanceSteps)\r\n{\r\n\tCurDistance += StepSize;\r\n\tfor (int j = 0; j < (int) RadialSteps; j++)\r\n\t{\r\n\t\tSubOffset +=1;\r\n\t\tCurOffset.x = cos(TWOPI*((TempAARotation+SubOffset) / RadialSteps));\r\n\t\tCurOffset.y = sin(TWOPI*((TempAARotation+SubOffset) / RadialSteps));\r\n\t\tNewUV.x = UV.x + CurOffset.x * (CurDistance+(RandomSamp*TempAADistance));\r\n\t\tNewUV.y = UV.y + CurOffset.y * (CurDistance+(RandomSamp*TempAADistance));\r\n\t\tfloat distpow = pow(CurDistance, 0.1);\r\n\t\tAccumNormalDepth += Texture2DSample(NormalDepthTex,NormalDepthTexSampler,NewUV)*distpow;\t\t\r\n        AccumThickness += clamp((Texture2DSample(ThicknessTex,ThicknessTexSampler,NewUV).r),0,100000)*distpow;\t\r\n\t\tAccumdist += distpow;\r\n\t}\r\nSubOffset +=RadialOffset;\r\ni++;\r\n}\r\nAccumNormalDepth /= Accumdist;\r\nAccumThickness /= Accumdist;\r\nThicknessOutput = AccumThickness;\r\nreturn AccumNormalDepth;\r\n}\r\n\r\n"
   outputType=CMOT_Float4
   additionalOutputs [{outputName=ThicknessOutput, outputType=CMOT_Float1}]
   inputs [
    {
     inputName=NormalDepthTex
     input {expression=&self:MaterialExpressionTextureObjectParameter_0}
    }
    {inputName=ThicknessTex, input {expression=&self:MaterialExpressionTextureObjectParameter_1}}
    {inputName=UV, input {expression=&self:MaterialExpressionTextureCoordinate_0}}
    {inputName=Distance, input {expression=&self:MaterialExpressionMultiply_3}}
    {inputName=DistanceSteps, input {expression=&self:MaterialExpressionConstant_34}}
    {inputName=RadialSteps, input {expression=&self:MaterialExpressionConstant_35}}
    {inputName=RadialOffset, input {expression=&self:MaterialExpressionConstant_5}}
    {inputName=RandomTex, input {expression=&self:MaterialExpressionNamedRerouteUsage_26}}
   ]
  }
  n18 = ScalarParameter() {defaultValue=100, group=Filter, parameterName=Filter_ThicknessDivide}
  NamedRerouteDeclaration(None=Saturate(None=Divide(A=@n17.ThicknessOutput, B=@n18) {constB=250})) {name=ThicknessOutput, nodeColor=<rgba 0 0.9766 1 1>}
  n19 = LinearInterpolate(
   A=VectorParameter() {group=Color, parameterName=LowThicknessColor, defaultValue=<rgba 0.9 0.9 0.9 1>}.RGB
   B=VectorParameter() {group=Color, parameterName=HighThicknessColor, defaultValue=<rgba 0.8 0.8 0.8 1>}.RGB
   Alpha=NamedRerouteUsage()
  )
  n20 = NamedRerouteUsage()
  n21 = OneMinus(None=ScalarParameter() {defaultValue=1, group=Color, parameterName=TranslucentFluid, sortPriority=0})
  NamedRerouteDeclaration(None=LinearInterpolate(A=Multiply(A=@n19, B=@n20), Alpha=@n21) {constB=0}) {name=ColoredSceneColor, nodeColor=<rgba 0.125 1 0 1>}
  n22 = VectorParameter() {
   group=Color
   parameterName=Shading
   sortPriority=36
   channelNames {a=Contrast, b=Offset, g=LightFace, r=DarkFace}
   defaultValue=<rgba 1 1 0.6 0.5>
  }
  n23 = DotProduct(A=SkyAtmosphereLightDirection(), B=NamedRerouteUsage())
  n24 = MaterialFunctionCall(In=Saturate(None=Add(A=@n23, B=@n22.Offset) {constB=0.6}), Contrast=@n22.Contrast) {materialFunction=&CheapContrast}
  n25 = VectorParameter() {group=Color, parameterName=HueShift, sortPriority=35, channelNames {a="", b="", g=LightFace, r=DarkFace}}
  n26 = LinearInterpolate(A=@n25.DarkFace, B=@n25.DarkFace, Alpha=@n24.Result)
  n27 = LinearInterpolate(B=@n19, Alpha=@n21) {constB=0}
  NamedRerouteDeclaration(None=Multiply(
    A=LinearInterpolate(A=@n22.DarkFace, B=@n22.DarkFace, Alpha=@n24.Result)
    B=MaterialFunctionCall(Hue Shift Percentage=@n26, Texture=@n27) {materialFunction=&HueShift}.Result
   )) {name=BaseColor, nodeColor=<rgba 0 0.2031 1 1>}
  MaterialExpressionWorldPosition_4 = WorldPosition() {worldPositionShaderOffset=WPT_ExcludeAllShaderOffsets}
  MaterialExpressionTextureObjectParameter_4 = TextureObjectParameter() {group=FromNiagara, parameterName=NormalDepthTex}
  MaterialExpressionSkyAtmosphereLightDirection_3 = SkyAtmosphereLightDirection()
  MaterialExpressionConstant_32 = Constant() {r=30}
  MaterialExpressionCameraPositionWS_5 = CameraPositionWS()
  MaterialExpressionMaterialFunctionCall_13 = MaterialFunctionCall() {materialFunction=&WorldPositionBehindTranslucency}
  MaterialExpressionConstant_33 = Constant() {r=50}
  n28 = VectorParameter() {group=FromNiagara, parameterName=WorldToUnit_Row0}
  MaterialExpressionAppendVector_0 = AppendVector(A=@n28.RGB, B=@n28.RGB)
  n29 = VectorParameter() {group=FromNiagara, parameterName=WorldToUnit_Row1}
  MaterialExpressionAppendVector_1 = AppendVector(A=@n29.RGB, B=@n29.RGB)
  n30 = VectorParameter() {group=FromNiagara, parameterName=WorldToUnit_Row2}
  MaterialExpressionAppendVector_2 = AppendVector(A=@n30.RGB, B=@n30.RGB)
  n31 = VectorParameter() {group=FromNiagara, parameterName=WorldToUnit_Row3}
  MaterialExpressionAppendVector_3 = AppendVector(A=@n31.RGB, B=@n31.RGB)
  MaterialExpressionMaterialFunctionCall_19 = MaterialFunctionCall() {materialFunction=&CameraDirectionVector}
  MaterialExpressionNamedRerouteUsage_1 = NamedRerouteUsage()
  n32 = Custom!(
   WorldPosition=@MaterialExpressionWorldPosition_4.XYZ
   LiquidDepthTexture=@MaterialExpressionTextureObjectParameter_4
   LightDirection=@MaterialExpressionSkyAtmosphereLightDirection_3
   RayMarchingDistance=@MaterialExpressionConstant_32
   CameraPosition=@MaterialExpressionCameraPositionWS_5
   WorldPositionBehindTranslucent=@MaterialExpressionMaterialFunctionCall_13.Result
   MaxDistance=@MaterialExpressionConstant_33
   WTU_A=@MaterialExpressionAppendVector_0
   WTU_B=@MaterialExpressionAppendVector_1
   WTU_C=@MaterialExpressionAppendVector_2
   WTU_D=@MaterialExpressionAppendVector_3
   CameraDirection=@MaterialExpressionMaterialFunctionCall_19.Result
   RandomTex=@MaterialExpressionNamedRerouteUsage_1
  ) {
   code="##define LOOPCOUNT 5\r\n\r\nfloat4x4 WorldToUnit = {WTU_A,WTU_B,WTU_C,WTU_D};\r\n\r\nfloat2 ScreenUV = MaterialFloat2(ScreenAlignedPosition(Parameters.ScreenPosition).xy);\r\nfloat2 ScenePixels=View.BufferSizeAndInvSize.xy*ScreenUV;\r\nScenePixels+=View.TemporalAAParams.r;\r\nfloat2 RandomSamp = ((uint)(ScenePixels.x) + 2 * (uint)(ScenePixels.y)) % 5;\r\nRandomSamp+=Texture2DSample(RandomTex,RandomTexSampler,ScenePixels);\r\nRandomSamp/=5;\r\n\r\nfloat StepDistance = RayMarchingDistance/LOOPCOUNT*lerp(1.0,RandomSamp.x,0.5);\r\n\r\nfloat theta = 2.0 * 3.14159265 * saturate(RandomSamp.x);\r\nfloat r = sqrt(saturate(RandomSamp.x));\r\nfloat x = r * cos(theta);float y = r * sin(theta);\r\nfloat3 referenceVec = normalize(abs(LightDirection.x) > abs(LightDirection.y) ? float3(-LightDirection.z, 0, LightDirection.x) : float3(0, LightDirection.z, -LightDirection.y));\r\nfloat3 orthogonalVec = normalize(cross(LightDirection, referenceVec));\r\nfloat3 randomPoint = x * referenceVec + y * orthogonalVec;\r\nfloat z = lerp(1, cos(0.5), saturate(RandomSamp.x));\r\nfloat3 randomVec = normalize(z * LightDirection + sqrt(1 - z * z) * randomPoint);\r\n\r\nfloat InShadow = 0;\r\nfor (int i = 1;i<LOOPCOUNT;i++){\r\n\r\n    float3 RayMarchingVector = StepDistance*i*randomVec;\r\n    float3 RayMarchingPosition = WorldPositionBehindTranslucent+RayMarchingVector;\r\n    float3 RayMarchingVector_P=(RayMarchingVector-dot(RayMarchingVector,CameraDirection)*CameraDirection)*(WorldPosition-CameraPosition)/(WorldPositionBehindTranslucent-CameraPosition);\r\n    \r\n    float4 RayMarchingPosition_U = mul(float4(WorldPosition+RayMarchingVector_P,1.0),WorldToUnit);\r\n    float2 RayMarchingSampleUV = RayMarchingPosition_U.xy;\r\n    float LiquidDepth = Texture2DSampleLevel(LiquidDepthTexture,LiquidDepthTextureSampler,RayMarchingSampleUV,0.0).a;\r\n    \r\n    float RayMarchingPositionDepth = dot(RayMarchingPosition-CameraPosition,CameraDirection);\r\n    float DepthDistance = RayMarchingPositionDepth - LiquidDepth;\r\n\r\n    InShadow = saturate(ceil(DepthDistance)) * saturate((MaxDistance-DepthDistance)) * (1-(float)i /LOOPCOUNT);\r\n\r\n    if(InShadow>0)\r\n    {\r\n        break; \r\n    }\r\n}\r\nreturn InShadow;"
   inputs [
    {
     inputName=WorldPosition
     input {expression=&self:MaterialExpressionWorldPosition_4, mask=1, maskB=1, maskG=1, maskR=1}
    }
    {
     inputName=LiquidDepthTexture
     input {expression=&self:MaterialExpressionTextureObjectParameter_4}
    }
    {
     inputName=LightDirection
     input {expression=&self:MaterialExpressionSkyAtmosphereLightDirection_3}
    }
    {inputName=RayMarchingDistance, input {expression=&self:MaterialExpressionConstant_32}}
    {inputName=CameraPosition, input {expression=&self:MaterialExpressionCameraPositionWS_5}}
    {
     inputName=WorldPositionBehindTranslucent
     input {expression=&self:MaterialExpressionMaterialFunctionCall_13}
    }
    {inputName=MaxDistance, input {expression=&self:MaterialExpressionConstant_33}}
    {inputName=WTU_A, input {expression=&self:MaterialExpressionAppendVector_0}}
    {inputName=WTU_B, input {expression=&self:MaterialExpressionAppendVector_1}}
    {inputName=WTU_C, input {expression=&self:MaterialExpressionAppendVector_2}}
    {inputName=WTU_D, input {expression=&self:MaterialExpressionAppendVector_3}}
    {
     inputName=CameraDirection
     input {expression=&self:MaterialExpressionMaterialFunctionCall_19}
    }
    {inputName=RandomTex, input {expression=&self:MaterialExpressionNamedRerouteUsage_1}}
   ]
  }
  n33 = Saturate(None=Subtract(A=@n32, B=NamedRerouteUsage()))
  n34 = Multiply(
   A=ScalarParameter() {defaultValue=0.1, group=Shadow, parameterName=ShadowStrength}
   B=ScalarParameter() {defaultValue=1, group=Opacity, parameterName=OpacityScale}
  )
  NamedRerouteDeclaration(None=StaticSwitchParameter(True=Constant(), False=Multiply(A=@n33, B=@n34) {constB=0.75}) {defaultValue=true, parameterName=DisableShadow}) {name=Shadow, nodeColor=<rgba 0 0.7656 1 1>}
  NamedRerouteDeclaration(None=ComponentMask(None=@n17.return) {b=true, g=true, r=true}) {name=FluidNormal, nodeColor=<rgba 0.4297 1 0 1>}
  NamedRerouteDeclaration(None=ComponentMask(None=@n17.return) {a=true}) {name=FluidDepth, nodeColor=<rgba 0 0.1797 1 1>}
  n35 = NamedRerouteUsage()
  n36 = Saturate(None=Subtract(A=Divide(A=@n35) {constB=40960}, B=ScalarParameter() {group=Opacity, parameterName=FluidEdgeOffset}))
  NamedRerouteDeclaration(None=OneMinus(None=Floor(None=@n36))) {name=FluidOpacity, nodeColor=<rgba 0.5703 0 1 1>}
  n37 = Multiply(
   A=LinearInterpolate(
    A=ScalarParameter() {defaultValue=10, group=Refraction, parameterName=RefractionStr_LowThickness}
    B=ScalarParameter() {defaultValue=80, group=Refraction, parameterName=RefractionStr_HighThickness}
    Alpha=NamedRerouteUsage()
   )
   B=ScalarParameter() {defaultValue=1, group=Opacity, parameterName=OpacityScale}
  )
  n38 = MaterialFunctionCall!(Ray Direction=CameraVectorWS(), Surface Normal=NamedRerouteUsage(), Refractive Index Origin, Refractive Index Target=Constant() {r=1.03}) {materialFunction=/Engine/Functions/Engine_MaterialFunctions01/Vectors/Refract}
  n39 = Subtract(A=Add(A=Multiply(A=@n37, B=@n38.Vector), B=WorldPosition().XYZ), B=CameraPositionWS())
  NamedRerouteDeclaration(None=SceneColor!(UVs=MaterialFunctionCall!(Camera Relative Position=@n39, X Camera to World Vector, Y Camera to World Vector, Z Camera to World Vector, "Tan(FOV/2)*[1,Screen Res Y/X]") {materialFunction=/Engine/Functions/Engine_MaterialFunctions02/Math/ManualWorldToScreenUVsTransform}.Screen UVs).RGB) {name=SampledSceneColor, nodeColor=<rgba 0 1 0.9297 1>}
  n40 = TextureSampleParameterCube!(UVs=ReflectionVectorWS(CustomWorldNormal=NamedRerouteUsage()), Apply View MipBias) {
   group=Reflection
   parameterName=HDRIReflectionTexture
   samplerType=SAMPLERTYPE_LinearColor
   texture=/NiagaraFluids/Textures/T_WaterReflection
  }
  n41 = VectorParameter() {group=Reflection, parameterName=HDRIReflectionTint, defaultValue=<rgba 1 1 1 1>}
  n42 = VectorParameter() {group=Reflection, parameterName=HDRIReflectionFresnel, channelNames {a="", b="", g=Base, r=Exp}, defaultValue=<rgba 2 0.2 0 1>}
  n43 = NamedRerouteUsage()
  NamedRerouteDeclaration(None=Multiply(A=Multiply(A=@n40.RGB, B=@n41.RGB), B=Fresnel(ExponentIn=@n42.Exp, BaseReflectFractionIn=@n42.Exp, Normal=@n43))) {name=HDRIReflection, nodeColor=<rgba 1 0 0.0469 1>}
  n44 = VectorParameter() {group=Reflection, parameterName=EnvReflectionTint, defaultValue=<rgba 0.125 0.125 0.125 1>}
  n45 = SkyLightEnvMapSample(Direction=ReflectionVectorWS(CustomWorldNormal=NamedRerouteUsage()))
  n46 = VectorParameter() {group=Reflection, parameterName=EnvReflectionFresnel, channelNames {a="", b="", g=Base, r=Exp}, defaultValue=<rgba 2 0 0 1>}
  n47 = NamedRerouteUsage()
  NamedRerouteDeclaration(None=LinearInterpolate(B=Multiply(A=@n44.RGB, B=@n45), Alpha=Fresnel(ExponentIn=@n46.Exp, BaseReflectFractionIn=@n46.Exp, Normal=@n47))) {name=EnvReflection, nodeColor=<rgba 0 1 0.6484 1>}
  n48 = TextureCoordinate()
  n49 = Constant2Vector() {g=1400, r=1400}
  n50 = ParticleSize()
  n51 = MaterialFunctionCall!(OffsetAmount=Multiply(A=Divide(A=@n49, B=@n50) {constA=1400, constB=1400}, B=Constant() {r=0.01})) {materialFunction=/Engine/Functions/Engine_MaterialFunctions01/ImageAdjustment/BlurSampleOffsets}
  n52 = TextureObjectParameter() {group=FromNiagara, parameterName=NormalDepthTex}
  n53 = Min(A=TextureSample(UVs=Add(A=@n48, B=@n51."-.5,.5"), Tex=@n52).A, B=TextureSample(UVs=Add(A=@n48, B=@n51.".5,.5"), Tex=@n52).A)
  n54 = TextureSample(UVs=Add(A=@n48, B=@n51.".5,-.5"), Tex=@n52)
  n55 = Add(A=@n48, B=@n51."-.5,-.5")
  MaterialExpressionMin_4 = Min(A=Min(A=Min(A=@n53, B=@n54.A), B=TextureSample(UVs=@n55, Tex=@n52).A), B=TextureSample(Tex=@n52).A)
  n56 = Custom!(Opacity=@MaterialExpressionMin_4) {
   code="float mask = 0;\r\nif (Opacity >= 65000)\r\n{\r\nclip(-1);\r\n}\r\nreturn mask;"
   outputType=CMOT_Float1
   inputs [{inputName=Opacity, input {expression=&self:MaterialExpressionMin_4}}]
  }
  NamedRerouteDeclaration(None=@n56) {name=Clip, nodeColor=<rgba 1 0 0.0703 1>}
 ]
}
