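# Unlit translucent Niagara sprite material that shades a simple 2D grid fluid fire from a texture parameter.
# The red channel, scaled by a temperature multiplier, drives a BlackBody node for the fire color.
# The green channel, likely smoke density, tints a dark grey color parameter that is added to the blackbody result.
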Material /NiagaraFluids/Materials/Grid2D/M_Fluids_SimpleFire
graph {
 properties {bUsedWithNiagaraSprites=true, blendMode=BLEND_Translucent, pixelDepthOffsetMode=PDOM_Legacy, shadingModel=MSM_Unlit, usedShadingModels=Unlit}
 nodes [
  n1 = TextureSampleParameter2D() {parameterName=Tex}
  n2 = ScalarParameter() {defaultValue=0.3, parameterName=TemperatureMult}
  Add(A=BlackBody!(Temp=Multiply(A=@n1.R, B=@n2)), B=Multiply(A=@n1.G, B=VectorParameter() {defaultValue=<rgba 0.1094 0.1094 0.1094 0>}.RGB))
 ]
}
