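# Unlit translucent material for Niagara sprites that, by its name, renders simple smoke for Niagara Fluids Grid2D simulations.
# It samples the Tex texture parameter at the default UVs and uses only the red channel as a mask.
# That channel scales a dark vector-parameter tint for color and, separately, the OpacityMult scalar for opacity.
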Material /NiagaraFluids/Materials/Grid2D/M_Fluids_SimpleSmoke
graph {
 properties {bUsedWithNiagaraSprites=true, blendMode=BLEND_Translucent, pixelDepthOffsetMode=PDOM_Legacy, shadingModel=MSM_Unlit, usedShadingModels=Unlit}
 nodes [
  n1 = TextureSample(UVs=TextureCoordinate(), Tex=TextureObjectParameter() {parameterName=Tex})
  Multiply(A=VectorParameter() {defaultValue=<rgba 0.0677 0.063 0.0518 0>}.RGB, B=@n1.R)
  Multiply(A=@n1.R, B=ScalarParameter() {defaultValue=3, parameterName=OpacityMult})
 ]
}
